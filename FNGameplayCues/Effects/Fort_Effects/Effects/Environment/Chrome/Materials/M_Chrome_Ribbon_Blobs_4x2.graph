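# Masked, two-sided lit chrome-liquid blob material for Niagara ribbons, sprites and mesh particles, built around the LackLuster base function.
# Noise-offset UVs and clamped 4x2 blob textures provide shape and normals, with particle alpha and near-camera fade feeding masking.
# Chrome look branches combine panning swirl normals, Fresnel hue shift, curve-atlas reflections, a view-space sphere map and switchable phased emissive.
# Dynamic parameters drive emissive, timer, near fade and noise reduction.
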Material /FNGameplayCues/Effects/Fort_Effects/Effects/Environment/Chrome/Materials/M_Chrome_Ribbon_Blobs_4x2
graph {
 properties {
  bUsedWithNiagaraMeshParticles=true
  bUsedWithNiagaraRibbons=true
  bUsedWithNiagaraSprites=true
  blendMode=BLEND_Masked
  pixelDepthOffsetMode=PDOM_Legacy
  twoSided=true
  usedShadingModels=Default Lit
 }
 nodes [
  Constant() {r=1}
  n1 = NamedRerouteUsage()
  n2 = DynamicParameter() {parameterIndex=2, defaultValue=<rgba 0 0 0 0>, paramNames [Emissive,---,---,---]}
  MaterialFunctionCall(Emissive=Add(A=Multiply(A=@n1, B=@n2.Emissive), B=NamedRerouteUsage())) {materialFunction=&MF_FX_EmissiveAdjustments}
  n3 = Multiply(A=NamedRerouteUsage(), B=ParticleColor().A)
  MaterialFunctionCall(In=@n3, Contrast=ScalarParameter() {defaultValue=-1}) {materialFunction=&CheapContrast}
  MaterialFunctionCall!(
   bBlendMaterialAttributes
   MA_In
   UVsOverride
   Override Normals
   BlendMaterialAttributesAlpha
   BaseColorOverride
   WorldPositionOverride
   VertexNormalOverride
   SourceNormalsStrengthOverride
  ) {materialFunction=/Game/TechArt/Seasonal/LackLuster/MaterialFunctions/MF_LackLuster_Base}
  n4 = DynamicParameter() {parameterIndex=3, defaultValue=<rgba 0 0 0 0>, paramNames [Timer,MeshScale,NearFade,NotUse]}
  Multiply(
   A=VertexInterpolator(VS=MaterialFunctionCall(Distance=@n4.NearFade) {materialFunction=&Near_Camera_Fade}.Result).PS
   B=@n3
  )
  n5 = MaterialFunctionCall(UVs=Add(A=NamedRerouteUsage(), B=NamedRerouteUsage()), Texture Scale=Constant2Vector() {g=3.6, r=1.9}) {materialFunction=&ScaleUVsByCenter}
  n6 = TextureSampleParameter2D(UVs=@n5.UVs) {
   group=03.BaseShape
   parameterName=BaseShape_Texture
   samplerSource=SSM_Clamp_WorldGroupSettings
   samplerType=SAMPLERTYPE_Normal
   texture=/Game/Effects/Fort_Effects/Textures/Water/T_ChromeLiquid_4x2_02_N
  }
  Multiply(
   A=TextureSampleParameter2D(UVs=@n5.UVs) {parameterName=Flat_Texture, texture=/Game/Effects/Fort_Effects/Textures/Water/T_ChromeLiquid_4x2_02_D}.G
   B=OneMinus(None=Desaturation(None=@n6.RGB))
  )
  Multiply(A=NamedRerouteUsage(), B=DynamicParameter() {parameterIndex=2, defaultValue=<rgba 0 0 0 0>, paramNames [Emissive,---,---,---]}.Emissive)
  NamedRerouteDeclaration(None=Add(A=Constant2Vector() {g=-1.35, r=-0.6}, B=TextureCoordinate() {uTiling=2, vTiling=3.5})) {name=R_UV, nodeColor=<rgba 0 1 0.2266 1>}
  n7 = DynamicParameter() {parameterIndex=3, defaultValue=<rgba 0 0 0 0>, paramNames [Timer,MeshScale,NearFade,NotUse]}
  n8 = LinearInterpolate(A=Time() {bOverride_Period=true, period=360}, B=@n7.Timer, Alpha=Saturate(None=Ceil(None=@n7.Timer)))
  n9 = Add(A=@n8, B=Multiply(A=VertexColor().A, B=ScalarParameter() {defaultValue=10, group=02.GlobalTimeSetting, parameterName=RandomScale}))
  NamedRerouteDeclaration(None=StaticSwitchParameter(True=@n8, False=StaticSwitchParameter(True=@n8, False=@n9) {group=********, parameterName=DisableThriveSpecialColorFixing}) {group=********, parameterName=ChangeToRibbonParticleMode}) {name=R_TimeRandom, nodeColor=<rgba 0.1484 1 0 1>}
  n10 = VectorParameter() {parameterName=EdgeNormalScale, defaultValue=<rgba 0.18 0.18 1 0>}
  n11 = TextureSample(UVs=NamedRerouteUsage()) {
   samplerSource=SSM_Wrap_WorldGroupSettings
   samplerType=SAMPLERTYPE_Normal
   texture=/Game/TechArt/Seasonal/LackLuster/Textures/T_Lackluster_Swirl_N
  }
  n12 = TextureSample(UVs=NamedRerouteUsage()) {
   samplerSource=SSM_Wrap_WorldGroupSettings
   samplerType=SAMPLERTYPE_Normal
   texture=&T_Lackluster_Swirl_N
  }
  n13 = Normalize(VectorInput=Add(A=@n11.RGB, B=@n12.RGB))
  NamedRerouteDeclaration(None=Transform(None=Multiply(A=@n10.RGB, B=@n13))) {name=ChromeNormalsWS, nodeColor=<rgba 1 0 0.5391 1>}
  NamedRerouteDeclaration(None=Divide(A=TextureCoordinate(), B=ScalarParameter() {defaultValue=1680, parameterName=Chrome_Scale_UV_a})) {name=Chrome_UVs_A_NoPanning_1, nodeColor=<rgba 0 1 0.7656 1>}
  n14 = ScalarParameter() {defaultValue=2.5, parameterName=EdgeNormal_FresnelExponent}
  n15 = VertexNormalWS()
  NamedRerouteDeclaration(None=LinearInterpolate(Alpha=MaterialFunctionCall(Exponent=@n14, Vertex / Pixel Normal=@n15) {materialFunction=&MF_Fresnel_FX}.Fresnel Inverted) {constA=0.05}) {name=EdgeNormalFresnelInverted, nodeColor=<rgba 0 1 0.5547 1>}
  NamedRerouteDeclaration(None=LinearInterpolate(
    A=ScalarParameter() {defaultValue=0.28, parameterName=Roughness.-Z}
    B=ScalarParameter() {defaultValue=0.22, parameterName=Roughness}
    Alpha=NamedRerouteUsage()
   )) {name=Roughness, nodeColor=<rgba 1 0.3984 0 1>}
  n16 = DotProduct(A=VertexNormalWS(), B=Constant3Vector() {constant=<rgba 0 0 -1 0>})
  NamedRerouteDeclaration(None=SphereMask(
    A=Saturate(None=@n16)
    B=ScalarParameter() {defaultValue=0.1, parameterName=VertexNormalMaskCenter}
    Radius=ScalarParameter() {defaultValue=0.2, parameterName=VertexNormalMaskRadius}
    Hardness=ScalarParameter() {defaultValue=0.5, parameterName=VertexNormalMaskHardness}
   )) {name=VertexNormalMask.-Z, nodeColor=<rgba 0 0.3203 1 1>}
  n17 = LinearInterpolate(Alpha=@n6.B) {constA=500, constB=0}
  NamedRerouteDeclaration(None=Add(A=Add(A=@n6.R, B=@n6.R), B=Clamp(None=@n17))) {name=R_BaseShapeSampledValue, nodeColor=<rgba 0.6466 0.4328 1 1>}
  n18 = MaterialFunctionCall(Float3=SceneTexture() {sceneTextureId=PPI_CustomDepth}.Color) {materialFunction=&BreakOutFloat3Components}
  n19 = PixelDepth()
  n20 = ComponentMask(None=@n19.R) {r=true}
  n21 = Saturate(None=Divide(A=Subtract(A=@n18.R, B=@n20), B=ScalarParameter() {defaultValue=50, group=05.CustomDepthCullAndFade, parameterName=CustomDepthFadeDistance}) {constB=50})
  n22 = Constant() {r=1}
  n23 = If(A=@n19.R, B=@n18.R, "A > B"=Constant(), "A == B"=@n22, "A < B"=@n22) {equalsThreshold=0.1}
  n24 = Constant() {r=1}
  NamedRerouteDeclaration(None=ShadingPathSwitch(Default=Multiply(A=@n21, B=@n23), Forward=@n24, Mobile=@n24)) {name=R_CustomDepthCullAndFade, nodeColor=<rgba 0 0.6953 1 1>}
  n25 = Add(A=TextureCoordinate() {uTiling=0.4, vTiling=0.4}, B=ParticleRandom())
  n26 = AppendVector(A=ScalarParameter() {defaultValue=0.01, parameterName=X_WobbleSpeed}, B=ScalarParameter() {defaultValue=0.1, parameterName=Y_WobbleSpeed})
  n27 = ComponentMask(None=TextureSample(UVs=Panner(Coordinate=@n25, Speed=@n26) {speedY=0.4}) {samplerType=SAMPLERTYPE_Normal, texture=/Game/Global/Textures/Default/Noise/T_SwirlNoise_N}.RGB) {g=true, r=true}
  n28 = ScalarParameter() {defaultValue=0.2, parameterName=WiggleIntensity}
  n29 = Clamp(None=Add(A=TextureCoordinate(), B=Multiply(A=@n27, B=@n28)))
  n30 = VectorParameter() {parameterName=ScaleNormal, defaultValue=<rgba 0.5 0.5 2 0>}
  n31 = VectorParameter() {parameterName=ScaleNormalHigh, defaultValue=<rgba 2 2 1 0>}
  n32 = MaterialFunctionCall() {materialFunction=&DebugTimeSine}
  n33 = ReflectionVectorWS(CustomWorldNormal=Multiply(
    A=TextureSample(UVs=@n29) {samplerType=SAMPLERTYPE_Normal, texture=/Game/Effects/Fort_Effects/Textures/Water/T_ChromeLiquid_4x2_N}.RGB
    B=LinearInterpolate(A=@n30.RGB, B=@n31.RGB, Alpha=@n32.Result)
   ))
  n34 = ConstantBiasScale(None=ComponentMask(None=Transform(None=@n33) {transformSourceType=TRANSFORMSOURCE_World, transformType=TRANSFORM_View}) {g=true, r=true})
  n35 = ComponentMask(None=VectorParameter() {parameterName=CubeMapOffset, defaultValue=<rgba 1 1 1 1>}.RGB) {g=true, r=true}
  n36 = TextureSampleParameter2D(UVs=Clamp(None=Multiply(A=@n34, B=@n35))) {
   parameterName=SphereReflectionMap
   samplerType=SAMPLERTYPE_LinearGrayscale
   texture=/Game/Effects/Fort_Effects/Textures/T_Emote_Monarch_Sphere
  }
  n37 = Multiply(A=@n36.RGB, B=ScalarParameter() {defaultValue=0.5, group=reflection, parameterName=Reflection Multiplier})
  n38 = ComponentMask(None=MaterialFunctionCall() {materialFunction=/Engine/Functions/GetPostProcessFunctions/GetAmbientCubemapTint}.AmbientCubemapTint) {b=true, g=true, r=true}
  n39 = Multiply(A=Clamp(None=Multiply(A=@n37, B=@n38)) {maxDefault=15}, B=Multiply(A=ParticleColor().RGB, B=Constant() {r=1}))
  n40 = Add(
   A=Divide(A=VectorParameter() {parameterName=Diffuse_ColorDodgeBase, defaultValue=<rgba 0.4601 0.4251 0.4002 0>}.RGB, B=OneMinus(None=@n39))
   B=ScalarParameter() {defaultValue=-0.5, parameterName=DiffuseBias}
  )
  n41 = ScalarParameter() {defaultValue=1.6, parameterName=DiffuseScale}
  NamedRerouteDeclaration(None=Saturate(None=Multiply(A=@n40, B=@n41))) {name=Diffuse, nodeColor=<rgba 1 0 0.1172 1>}
  n42 = VectorParameter() {
   group=01.UVDistortNoise
   parameterName=NoiseA_UVTransform
   sortPriority=30
   channelNames {a=PannerV, b=PannerU, g=ScaleV, r=ScaleU}
   defaultValue=<rgba 1 1 0.1 0.5>
  }
  n43 = Multiply(A=NamedRerouteUsage(), B=AppendVector(A=@n42.ScaleU, B=@n42.ScaleU))
  n44 = VectorParameter() {
   group=01.UVDistortNoise
   parameterName=NoiseA_UVOffset
   sortPriority=31
   channelNames {a="", b="", g=OffsetV, r=OffsetU}
   defaultValue=<rgba 0 0 1 1>
  }
  n45 = AppendVector(A=@n44.OffsetU, B=@n44.OffsetU)
  n46 = TextureSampleParameter2D(UVs=Panner(Coordinate=Add(A=@n43, B=@n45), Speed=AppendVector(A=@n42.PannerU, B=@n42.PannerU)) {speedY=0.2}) {
   group=01.UVDistortNoise
   parameterName=NoiseA_Texture
   texture=/Game/CharClasses/Outlander/Gadgets/AntiMaterialCharge/Materials/T_AntiMaterialCharge_Dissolve_01
  }
  n47 = AppendVector(A=@n46.G, B=@n46.G)
  n48 = VectorParameter() {
   group=01.UVDistortNoise
   parameterName=NoiseA_Property
   channelNames {a=NoiseStrength_V, b=NoiseStrength_U, g=NoiseStrengthOffset_V, r=NoiseStrengthOffset_U}
   defaultValue=<rgba -0.25 0.5 0.1 0.5>
  }
  n49 = AppendVector(A=@n48.NoiseStrengthOffset_U, B=@n48.NoiseStrengthOffset_U)
  NamedRerouteDeclaration(None=Multiply(A=Add(A=@n47, B=@n49) {constB=-0.5}, B=AppendVector(A=@n48.NoiseStrength_U, B=@n48.NoiseStrength_U))) {name=R_NoiseA, nodeColor=<rgba 0 1 0.25 1>}
  n50 = ComponentMask(None=ReflectionVectorWS()) {b=true, r=true}
  n51 = Add(A=Divide(A=TextureCoordinate(), B=Constant2Vector() {g=1400, r=2600}) {constB=1400}, B=Power(Base=Abs(None=@n50)) {constExponent=3})
  n52 = TextureSample(UVs=Panner(Coordinate=@n51) {bFractionalPart=true, speedX=0.7056, speedY=-0.0154}) {samplerSource=SSM_Wrap_WorldGroupSettings, texture=/Game/TechArt/Seasonal/LackLuster/Textures/T_Lackluster_Phased}
  n53 = Constant2Vector() {g=-1, r=-0.66}
  n54 = TextureSample(UVs=Panner(Coordinate=Multiply(A=@n51, B=@n53)) {bFractionalPart=true, speedX=-1.2908, speedY=-0.0195}) {samplerSource=SSM_Wrap_WorldGroupSettings, texture=&T_Lackluster_Phased}
  n55 = ScalarParameter() {defaultValue=0.2, parameterName=PhaseEmissiveScaleLow}
  n56 = ScalarParameter() {defaultValue=0.5, parameterName=PhaseEmissiveScaleHigh}
  n57 = MaterialFunctionCall(Exponent=ScalarParameter() {defaultValue=2, parameterName=PhaseFresnelExponent}) {materialFunction=&MF_Fresnel_FX}
  n58 = ScalarParameter() {parameterName=PhaseActive}
  n59 = Multiply(A=Multiply(A=Add(A=@n52.RGB, B=@n54.RGB), B=LinearInterpolate(A=@n55, B=@n56, Alpha=@n57.Fresnel Inverted)), B=@n58)
  n60 = Multiply(A=@n51, B=Constant2Vector() {g=-1, r=1}) {constB=-1}
  n61 = Max(
   A=TextureSample(UVs=Panner(Coordinate=@n51) {speedX=0.061, speedY=0.01}) {samplerSource=SSM_Wrap_WorldGroupSettings, texture=&T_Lackluster_Phased}.RGB
   B=TextureSample(UVs=Panner(Coordinate=@n60) {bFractionalPart=true, speedX=0.045, speedY=0.013}) {samplerSource=SSM_Wrap_WorldGroupSettings, texture=&T_Lackluster_Phased}.RGB
  )
  n62 = ScalarParameter() {defaultValue=0.5, parameterName=PhaseBaseDesaturation}
  n63 = Multiply(
   A=OneMinus(None=@n58)
   B=Multiply(A=Desaturation(None=@n61, Fraction=@n62), B=ScalarParameter() {defaultValue=0.1, parameterName=PhaseBaseEmissiveScale})
  )
  NamedRerouteDeclaration(None=StaticSwitchParameter(True=Add(A=@n59, B=@n63), False=Constant()) {parameterName=IsPhased}) {name=Emissive, nodeColor=<rgba 1 0 0.8203 1>}
  n64 = NamedRerouteUsage()
  n65 = TextureSample(UVs=Add(A=ComponentMask(None=@n64) {g=true, r=true}, B=NamedRerouteUsage())) {samplerType=SAMPLERTYPE_LinearColor, texture=/Game/TechArt/Seasonal/LackLuster/Textures/T_Lackluster_Streaks}
  n66 = MaterialFunctionCall(Exponent=ScalarParameter() {defaultValue=1.5, parameterName=FresnelHueShiftExponent}) {materialFunction=&MF_Fresnel_FX}
  n67 = CurveAtlasRowParameter(CurveTime=Frac(None=Add(A=@n65.R, B=@n66.Fresnel))) {
   atlas=/Game/TechArt/Seasonal/LackLuster/Curves/CA_Lackluster
   curve=/Game/TechArt/Seasonal/LackLuster/Curves/C_Lackluster_Reflection
   parameterName=ReflectionCurve
  }
  n68 = Frac(None=Multiply(A=@n66.Fresnel, B=ScalarParameter() {defaultValue=1, parameterName=HueShiftScale}))
  n69 = VectorParameter() {parameterName=AlbedoHueShift, defaultValue=<rgba 0.8212 0.9028 1 0>}
  n70 = Add(
   A=Multiply(A=@n67, B=NamedRerouteUsage())
   B=Multiply(
    A=VectorParameter() {parameterName=AlbedoScale, defaultValue=<rgba 0.8385 0.8193 0.6886 0>}.RGB
    B=MaterialFunctionCall(Hue Shift Percentage=@n68, Texture=@n69.RGB) {materialFunction=&HueShift}.Result
   )
  )
  n71 = NamedRerouteUsage()
  NamedRerouteDeclaration(None=Desaturation(None=Add(A=@n70, B=@n71), Fraction=ScalarParameter() {defaultValue=-0.5, parameterName=Fraction})) {name=BaseColor, nodeColor=<rgba 0 0.6484 1 1>}
  n72 = VectorParameter() {
   group=01.UVDistortNoise
   parameterName=NoiseB_UVTransform
   sortPriority=30
   channelNames {a=PannerV, b=PannerU, g=ScaleV, r=ScaleU}
   defaultValue=<rgba 2 1 0.01 0.2>
  }
  n73 = Multiply(A=NamedRerouteUsage(), B=AppendVector(A=@n72.ScaleU, B=@n72.ScaleU))
  n74 = VectorParameter() {
   group=01.UVDistortNoise
   parameterName=NoiseB_UVOffset
   sortPriority=31
   channelNames {a="", b="", g=OffsetV, r=OffsetU}
   defaultValue=<rgba 1 1 1 0>
  }
  n75 = AppendVector(A=@n74.OffsetU, B=@n74.OffsetU)
  n76 = TextureSampleParameter2D(UVs=Panner(Coordinate=Add(A=@n73, B=@n75), Speed=AppendVector(A=@n72.PannerU, B=@n72.PannerU)) {speedX=0.216}) {
   group=01.UVDistortNoise
   parameterName=NoiseB_Texture
   texture=&T_AntiMaterialCharge_Dissolve_01
  }
  n77 = AppendVector(A=@n76.R, B=@n76.R)
  n78 = VectorParameter() {
   group=01.UVDistortNoise
   parameterName=NoiseB_Property
   channelNames {a=NoiseStrength_V, b=NoiseStrength_U, g=NoiseStrengthOffset_V, r=NoiseStrengthOffset_U}
   defaultValue=<rgba 0.2 0.2 0.1 0>
  }
  n79 = AppendVector(A=@n78.NoiseStrengthOffset_U, B=@n78.NoiseStrengthOffset_U)
  NamedRerouteDeclaration(None=Multiply(A=Add(A=@n77, B=@n79) {constB=-0.5}, B=AppendVector(A=@n78.NoiseStrength_U, B=@n78.NoiseStrength_U))) {name=R_NoiseB, nodeColor=<rgba 1 0.7969 0 1>}
  n80 = Saturate(None=DynamicParameter() {defaultValue=<rgba 0 0 0.5 1>, paramNames [BaseStretch,Param2,NoiseReducer,Tiler]}.NoiseReducer)
  n81 = NamedRerouteUsage()
  n82 = NamedRerouteUsage()
  NamedRerouteDeclaration(None=Multiply(A=OneMinus(None=@n80), B=Add(A=@n81, B=@n82))) {name=R_MixedNoise, nodeColor=<rgba 0 1 0.0859 1>}
  NamedRerouteDeclaration(None=@n6.RGB) {name=R_Normal, nodeColor=<rgba 0.6466 0.4328 1 1>}
  n83 = Divide(A=TextureCoordinate(), B=ScalarParameter() {defaultValue=1680, parameterName=Chrome_Scale_UV_a})
  n84 = Multiply(A=ScalarParameter() {defaultValue=1, parameterName=ChromePanSpeedScale}, B=Time())
  n85 = ScalarParameter() {parameterName=Xspeed_01}
  n86 = ScalarParameter() {parameterName=Yspeed_01}
  NamedRerouteDeclaration(None=Panner(
    Coordinate=NamedRerouteDeclaration(None=@n83) {name=Chrome_UVs_A_NoPanning, nodeColor=<rgba 0 1 0.7656 1>}
    Time=@n84
    Speed=AppendVector(A=@n85, B=@n86)
   ) {bFractionalPart=true, speedX=0.0008, speedY=0.0022}) {name=Chrome_UVs_A, nodeColor=<rgba 0.9219 0 1 1>}
  n87 = TextureCoordinate()
  n88 = ScalarParameter() {defaultValue=2352, parameterName=Chrome_Scale_UV_b}
  n89 = ScalarParameter() {parameterName=Xspeed_02}
  n90 = ScalarParameter() {parameterName=Yspeed_02}
  NamedRerouteDeclaration(None=Panner(Coordinate=Divide(A=@n87, B=@n88), Time=@n84, Speed=AppendVector(A=@n89, B=@n90)) {bFractionalPart=true, speedX=0.0054, speedY=0.0022}) {name=Chrome_UVs_B, nodeColor=<rgba 0 0.3906 1 1>}
  NamedRerouteUsage()
  NamedRerouteUsage()
  NamedRerouteUsage()
  NamedRerouteUsage()
  Normalize(VectorInput=Multiply(A=NamedRerouteUsage(), B=VectorParameter() {parameterName=NormalStrength, defaultValue=<rgba 15 15 1 1>}.RGB))
  OneMinus(None=@n6.G)
  OneMinus(None=@n6.R)
  OneMinus(None=@n6.B)
  ScalarParameter() {defaultValue=1, parameterName=NearCamFade}
 ]
}
